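# Niagara effect for the MoleAcorn backpack: a blue-toned energy burst of lightning meshes, flare, flash, smoke and lights, despite the Idle name.
# Four emitters share per-frame spawns gated by a value-state-change flag, skipping the first frame.
# Negative starting ages derived from system build-up and fade-back times delay each layer; spawn groups split flare and lightning variations.
# User.VariantIndex blends flare and glow colors; User.KillCount is exposed.
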NiagaraSystem /BRCosmetics/Accessories/FORT_Backpacks/Backpack_M_MED_MoleAcornPecan/FX/NS_MoleAcorn_BackPack_Idle
dependencies {
 usedDataInterfaces [&NiagaraDataInterfaceCurve, &NiagaraDataInterfaceColorCurve]
 usedDynamicInputs []
 usedModules [
  &self:SystemSpawnScript.SystemScriptSource.SystemScriptGraph.NiagaraNodeAssignment_0.SetVariables_#1
  &self:SystemSpawnScript.SystemScriptSource.SystemScriptGraph.NiagaraNodeAssignment_1.SetVariables_#2
  &self:SystemSpawnScript.SystemScriptSource.SystemScriptGraph.NiagaraNodeAssignment_3.SetVariables_#3
  &self:SystemSpawnScript.SystemScriptSource.SystemScriptGraph.NiagaraNodeAssignment_7.SetVariables_#4
  &self:SystemSpawnScript.SystemScriptSource.SystemScriptGraph.NiagaraNodeAssignment_10.SetVariables_#5
  &SystemState
  /Game/Effects/Niagara/Modules/Logic/UpdateValueStateChange
 ]
}
emitters {
 NE_ExplosionLightnings {
  emitterUpdateScript [
   &EmitterState()
   &SpawnPerFrame(
    Spawn:NiagaraBool=&Boolean_AndOperation(
     Bool1=&Boolean_NotOperation(Bool=$System.UpdateValueStateChange.Changed)
     Bool2=&Boolean_NotOperation(Bool=$Transient.FirstFrame)
    )
    Spawn Count=7
    Use Spawn Probability=false
   )
   &SpawnPerFrame(
    Spawn:NiagaraBool=&Boolean_AndOperation(
     Bool1=&Boolean_NotOperation(Bool=$System.UpdateValueStateChange.Changed)
     Bool2=&Boolean_NotOperation(Bool=$Transient.FirstFrame)
    )
    Spawn Count=12
    Spawn Group=1
    Use Spawn Probability=false
   ) {module=SpawnPerFrame001}
  ]
  particleSpawnScript [
   /SetVariables(Particles.MeshIndex=&RandomRangeInt(Minimum=0, Maximum=0)) {isSetParameters=true}
   /SetVariables(Particles.InitialExecIndex:NiagaraFloat=&ReturnNormalizedExecIndex(Normalized Index Includes One=false)) {isSetParameters=true}
   /SetVariables(Particles.VisibilityTag=&MaskIntBySpawnGroup(
     Masked Int=&MakeIntFromFloat(Float=&UniformAOrBFloat(A=0, Distribution Weight=0.75, Evaluation Type=::NewEnumerator0, Recalculate Random Each Loop=true, Override Seed=false))
     Passthrough Int=&MakeIntFromFloat(Float=&UniformAOrBFloat(A=0, Distribution Weight=1, Evaluation Type=::NewEnumerator0, Recalculate Random Each Loop=true, Override Seed=false))
     Spawn Group Mask=0
    )) {isSetParameters=true}
   &InitializeParticle(
    Color=&ColorFromCurve(
     ColorCurve=DataInterface() {propertyValues {
       AlphaCurve {keys [<at 0 1 RCIM_Cubic>,<at 1 0 RCIM_Cubic>]}
       BlueCurve {keys [<at 0.6173 3 RCIM_Cubic>]}
       GreenCurve {keys [<at 0.6173 1.5991 RCIM_Cubic>]}
       RedCurve {keys [<at 0.6173 0.4844 RCIM_Cubic>]}
      }}
     CurveIndex=&RandomRangeFloat(Minimum=0, Maximum=1)
    )
    Lifetime=&RandomRangeFloat(Minimum=0.5, Maximum=0.6)
    Mesh Scale=&Multiply_VectorByFloat(
     Vector=&MaskVector(Mask Source=1, Mask Parameter=$Particles.VisibilityTag, Mask=0, Masked Vector=<1 1 1>, Passthrough Vector=<0.8 0.8 1>)
     Float=&Multiply_Float(
      A=&MaskFloat(
       Mask Source=1
       Mask Parameter=$Particles.VisibilityTag
       Mask=0
       Masked Float=&RandomRangeFloat(Minimum=0.17, Maximum=0.2)
       Passthrough Float=&RandomRangeFloat(Minimum=0.3, Maximum=0.4)
      )
      B=&MaskFloatBySpawnGroup(Masked Float=0.4, Passthrough Float=0.5)
     )
    )
    Mesh Scale Mode=::NewEnumerator4
    UsePositionOffset=false
    Write Material Random=true
   )
   /SetVariables(Particles.Age:NiagaraFloat=&Add_Float(
     A=&Add_Float(
      A=&Multiply_Float(A=$System.BuildUpTime, B=-0.9)
      B=&RandomRangeFloat(Minimum=-0.2, Maximum=0)
     )
     B=&MaskFloatBySpawnGroup(Masked Float=0, Passthrough Float=&RandomRangeFloat(Minimum=0, Maximum=-0.25))
    )) {isSetParameters=true}
   &ShapeLocation(
    MaskSpawnGroup=false
    Offset=$System.BaseOffset
    Offset Mode=::NewEnumerator0
    Sphere Distribution=::NewEnumerator1
    Sphere Radius=&MaskFloatBySpawnGroup(Masked Float=15, Passthrough Float=1)
    U Position=&Multiply_Float(A=$Particles.InitialExecIndex, B=4)
    UseOwnerScale=true
    V Position=&Multiply_Float(A=$Particles.InitialExecIndex, B=3)
   )
   &InitialMeshOrientation(
    Mesh Axis To Orient=&MaskVector(Mask Source=1, Mask Parameter=$Particles.VisibilityTag, Mask=0, Masked Vector=<-1 0 0>, Passthrough Vector=<0 0 1>)
    Mesh Orientation Mode=::NewEnumerator2
    Orientation Vector=&MaskVectorBySpawnGroup(Masked Vector=$Particles.ShapeLocation.ShapeNormal, Passthrough Vector=$Particles.ShapeLocation.ShapeNormal)
    Use Rotation Vector=false
   )
  ]
  particleUpdateScript [
   &ParticleState()
   &ScaleMeshSize(Scale Factor=&VectorFromFloat(Value=&Lerp_Float(
      A=1
      B=1.5
      Alpha=&FloatFromCurve(FloatCurve=DataInterface() {propertyValues {Curve {keys [
           {arriveTangent=2.2251, interpMode=RCIM_Cubic, leaveTangent=2.2251, tangentMode=RCTM_User, time=0, value=0}
           {interpMode=RCIM_Cubic, tangentMode=RCTM_User, time=1, value=1}
          ]}}})
     )))
   &DynamicMaterialParameters(
    Index 0 Param 4=13
    Index 1 Param 1=&FloatFromCurve(FloatCurve=DataInterface() {propertyValues {Curve {keys [<at 0 0.5>,<at 1 1>]}}})
    Index 1 Param 3=&Lerp_Float(
     A=-5
     B=1
     Alpha=&FloatFromCurve(FloatCurve=DataInterface() {propertyValues {Curve {keys [<at 0 0>,<at 1 1>]}}})
    )
    Index 1 Param 4=10
    Index 2 Param 1=&Multiply_Float(
     A=&MaskFloat(
      Mask Source=1
      Mask Parameter=$StackContext.VisibilityTag
      Mask=0
      Masked Float=&RandomRangeFloat(Minimum=0.25, Maximum=0.3)
      Passthrough Float=&RandomRangeFloat(Minimum=0.3, Maximum=0.4)
     )
     B=&Divide_Float(A=&MakeFloatFromVector(Vector=$Particles.Scale, Channel=::NewEnumerator0), B=10)
    )
    Index 3 Param 3=&FloatFromCurve(FloatCurve=DataInterface() {propertyValues {Curve {keys [
         {interpMode=RCIM_Cubic, leaveTangent=1.1034, tangentMode=RCTM_Break, time=0, value=-1}
         {arriveTangent=2.2399, tangentMode=RCTM_User, time=0.7, value=0.125}
        ]}}})
    Index 3 Param 4=&FloatFromCurve(FloatCurve=DataInterface() {propertyValues {Curve {keys [
         {interpMode=RCIM_Cubic, leaveTangent=3.0601, tangentMode=RCTM_Break, time=0.5, value=-0.5}
         {arriveTangent=3.2053, tangentMode=RCTM_User, time=1, value=1}
        ]}}})
    Module.Index0Param0Write=true
    Module.Index0Param1Write=true
    Module.Index0Param2Write=true
    Module.Index0Param3Write=true
    Module.Index1Param0Write=true
    Module.Index1Param1Write=true
    Module.Index1Param2Write=true
    Module.Index1Param3Write=true
    Module.Index2Param0Write=true
    Module.Index2Param1Write=true
    Module.Index2Param2Write=true
    Module.Index2Param3Write=true
    Module.Index3Param0Write=true
    Module.Index3Param1Write=true
    Module.Index3Param2Write=true
    Module.Index3Param3Write=true
    Write Parameter Index 1=true
    Write Parameter Index 2=true
    Write Parameter Index 3=true
   )
  ]
  rendererClasses [&NiagaraMeshRendererProperties]
  renderers [
   {rendererClass=&NiagaraMeshRendererProperties, rendererIndex=0}
   {rendererClass=&NiagaraMeshRendererProperties, rendererIndex=1}
  ]
 }
 NE_Flare {
  emitterUpdateScript [
   &EmitterState()
   &SpawnPerFrame(
    Spawn:NiagaraBool=&Boolean_AndOperation(
     Bool1=&Boolean_NotOperation(Bool=$System.UpdateValueStateChange.Changed)
     Bool2=&Boolean_NotOperation(Bool=$Transient.FirstFrame)
    )
    Use Spawn Probability=false
   )
   &SpawnPerFrame(
    Spawn:NiagaraBool=&Boolean_AndOperation(
     Bool1=&Boolean_NotOperation(Bool=$System.UpdateValueStateChange.Changed)
     Bool2=&Boolean_NotOperation(Bool=$Transient.FirstFrame)
    )
    Spawn Group=1
    Use Spawn Probability=false
   ) {module=SpawnPerFrame001}
  ]
  particleSpawnScript [
   &InitializeParticle(
    Color Mode=::NewEnumerator0
    Lifetime=&MaskFloatBySpawnGroup(Masked Float=0.8, Passthrough Float=1.8)
    Mesh Scale Mode=::NewEnumerator3
    Mesh Uniform Scale=0.75
    Position Offset=$System.BaseOffset
    Sprite Rotation Mode=::NewEnumerator2
    Sprite Size Mode=::NewEnumerator3
    Uniform Sprite Size=&MaskFloatBySpawnGroup(Masked Float=50, Passthrough Float=60)
    UsePositionOffset=true
    Write Material Random=true
   )
   /SetVariables(Particles.Age:NiagaraFloat=&MaskFloatBySpawnGroup(
     Masked Float=&Multiply_Float(A=$System.BuildUpTime, B=-0.9)
     Passthrough Float=&Multiply_Float(A=$System.FadeBackInTime, B=-1)
    )) {isSetParameters=true, module=SetVariables_#6}
   &InitialMeshOrientation(Orientation Axis=<0 0 1>, Orientation Vector=<0 0 1>, Rotation=<0 0.25 0>, Use Orientation Vector=true, Use Rotation Vector=true)
   &InitialMeshOrientation(
    Orientation Axis=<0 0 0>
    Orientation Vector=<0 0 0>
    Rotation=&RandomRangeVector(Minimum=<0 0 0>, Maximum=<1 0 0>)
    Use Orientation Vector=false
    Use Rotation Vector=true
   ) {module=InitialMeshOrientation001}
   &CameraOffset(Camera Offset Amount=20)
  ]
  particleUpdateScript [
   &ParticleState()
   &ScaleSpriteSize(
    Uniform Curve Scale=&MaskFloatBySpawnGroup(Masked Float=1.2, Passthrough Float=0.8)
    Uniform Curve Sprite Scale=DataInterface() {propertyValues {Curve {keys [
        {arriveTangent=2.0798, interpMode=RCIM_Cubic, leaveTangent=2.0798, tangentMode=RCTM_User, time=0, value=0}
        {interpMode=RCIM_Cubic, tangentMode=RCTM_User, time=1, value=1}
       ]}}}
   )
   &ScaleMeshSize(Scale Factor=&VectorFromFloat(Value=&MaskFloatBySpawnGroup(
      Masked Float=&FloatFromCurve(
       FloatCurve=DataInterface() {propertyValues {Curve {keys [
           {arriveTangent=2.3498, interpMode=RCIM_Cubic, leaveTangent=2.3498, tangentMode=RCTM_User, time=0, value=0}
           {arriveTangent=1, interpMode=RCIM_Cubic, leaveTangent=1, time=0.1, value=0.5}
           {interpMode=RCIM_Cubic, tangentMode=RCTM_User, time=1, value=1}
          ]}}}
       Scale Curve=0.5
      )
      Passthrough Float=&FloatFromCurve(
       FloatCurve=DataInterface() {propertyValues {Curve {keys [
           {arriveTangent=2.3498, interpMode=RCIM_Cubic, leaveTangent=2.3498, tangentMode=RCTM_User, time=0, value=0}
           {interpMode=RCIM_Cubic, tangentMode=RCTM_User, time=0.5, value=1}
           {arriveTangent=-0.3302, interpMode=RCIM_Cubic, leaveTangent=-0.3302, tangentMode=RCTM_User, time=1, value=0.9}
          ]}}}
       Scale Curve=0.3
      )
     )))
   &ScaleColor(
    Scale Alpha=&MaskFloatBySpawnGroup(
     Masked Float=&FloatFromCurve(FloatCurve=DataInterface() {propertyValues {Curve {keys [
          {arriveTangent=-0.8571, interpMode=RCIM_Cubic, leaveTangent=-0.8571, tangentMode=RCTM_User, time=0, value=1}
          {arriveTangent=0.062, interpMode=RCIM_Cubic, leaveTangent=0.062, tangentMode=RCTM_User, time=1, value=0}
         ]}}})
     Passthrough Float=&FloatFromCurve(
      FloatCurve=DataInterface() {propertyValues {Curve {keys [
          {arriveTangent=0.9089, interpMode=RCIM_Cubic, leaveTangent=0.9089, tangentMode=RCTM_User, time=0, value=0}
          {arriveTangent=0.062, interpMode=RCIM_Cubic, leaveTangent=0.062, tangentMode=RCTM_User, time=0.5, value=1}
          <at 1 0 RCIM_Cubic>
         ]}}}
      Scale Curve=0.1
     )
    )
    Scale RGB=&MakeVectorFromLinearColorRGB(Color=&Multiply_LinearColorByFloat(
      Linear Color=&ColorFromCurve(ColorCurve=DataInterface() {propertyValues {
         AlphaCurve {keys [<at 0 1 RCIM_Cubic>]}
         BlueCurve {keys [<at 0 50 RCIM_Cubic>,<at 0.9851 10 RCIM_Cubic>]}
         GreenCurve {keys [<at 0 0 RCIM_Cubic>,<at 0.9851 3.0139 RCIM_Cubic>]}
         RedCurve {keys [<at 0 0.2009 RCIM_Cubic>,<at 0.9851 0.5952 RCIM_Cubic>]}
        }})
      Float=1
     ))
    ScaleA=true
    ScaleRGB=true
   )
   /SetVariables(
    Particles.GlowColor=&MultiplyLinearColor(
     Linear Color=$Particles.Color
     Scale Factor=&MakeVector4_FromLinearColor(LinearColor=&LerpLinearColors(StartColor=<rgba 0 0.014 0.5 1>, EndColor=<rgba 0.5 0.2545 0.2689 1>, LerpFactor=$User.VariantIndex))
    )
    Particles.GlowSize=&Multiply_Vector2D_ByFloat(Float=1.25, Vector2D=$Particles.SpriteSize)
   ) {isSetParameters=true, module=SetVariables_#7}
   /SetVariables(Particles.LongFlareSize:Vector2f=&Multiply_Vector2D(A=$Particles.SpriteSize, B=<0.75 0.75>)) {isSetParameters=true, module=SetVariables_#8}
   &DynamicMaterialParameters(Module.Param0WriteEnabled=false, Module.Param1WriteEnabled=false, Module.Param2WriteEnabled=false, Module.Param3WriteEnabled=false)
   &FN_Mesh_Attributes(
    Color=&MultiplyLinearColor(
     Linear Color=&MaskLinearColorBySpawnGroup(
      Masked Linear Color=&ColorFromCurve(ColorCurve=DataInterface() {propertyValues {
         AlphaCurve {keys [<at 0 1 RCIM_Cubic>]}
         BlueCurve {keys [<at 0 50 RCIM_Cubic>,<at 1 1 RCIM_Cubic>]}
         GreenCurve {keys [<at 0 36.8983 RCIM_Cubic>,<at 1 0.1694 RCIM_Cubic>]}
         RedCurve {keys [<at 0 5.8837 RCIM_Cubic>,<at 1 0 RCIM_Cubic>]}
        }})
      Passthrough Linear Color=&ColorFromCurve(ColorCurve=DataInterface() {propertyValues {
         AlphaCurve {keys [<at 0 1 RCIM_Cubic>]}
         BlueCurve {keys [<at 0 0.5556 RCIM_Cubic>,<at 1 0.1136 RCIM_Cubic>]}
         GreenCurve {keys [<at 0 1.8155 RCIM_Cubic>,<at 1 0.4676 RCIM_Cubic>]}
         RedCurve {keys [<at 0 5 RCIM_Cubic>,<at 1 1 RCIM_Cubic>]}
        }})
     )
     Scale Factor=&MakeVector4FromVectorAndFloat(
      "Vector (RGB)"=&VectorFromFloat(Value=&MaskFloatBySpawnGroup(Masked Float=1, Passthrough Float=0.3))
      "Float (Alpha)"=&MaskFloatBySpawnGroup(
       Masked Float=&FloatFromCurve(FloatCurve=DataInterface() {propertyValues {Curve {keys [
            {interpMode=RCIM_Cubic, tangentMode=RCTM_User, time=0, value=1}
            {arriveTangent=-0.0487, interpMode=RCIM_Cubic, leaveTangent=-0.0487, tangentMode=RCTM_User, time=1, value=0}
           ]}}})
       Passthrough Float=&FloatFromCurve(FloatCurve=DataInterface() {propertyValues {Curve {keys [
            <at 0 0 RCIM_Cubic>
            {interpMode=RCIM_Cubic, tangentMode=RCTM_User, time=0.5, value=1}
            {arriveTangent=-0.0487, interpMode=RCIM_Cubic, leaveTangent=-0.0487, tangentMode=RCTM_User, time=1, value=0}
           ]}}})
      )
     )
    )
    Create Color Attribute=true
    Create Scale Attribute=true
    Scale=&Multiply_VectorByFloat(Vector=$Particles.Scale, Float=1.5)
    Write to Module Namespace=true
   )
   &FN_Mesh_Attributes(
    Color=&MultiplyLinearColor(
     Linear Color=&LerpLinearColors(
      StartColor=&ColorFromCurve(ColorCurve=DataInterface() {propertyValues {
         AlphaCurve {keys [<at 0 1 RCIM_Cubic>]}
         BlueCurve {keys [<at 0 10 RCIM_Cubic>,<at 1 1 RCIM_Cubic>]}
         GreenCurve {keys [<at 0 3.0139 RCIM_Cubic>,<at 1 0.3892 RCIM_Cubic>]}
         RedCurve {keys [<at 0 0.5952 RCIM_Cubic>,<at 1 0.2217 RCIM_Cubic>]}
        }})
      EndColor=&ColorFromCurve(ColorCurve=DataInterface() {propertyValues {
         AlphaCurve {keys [<at 0 1 RCIM_Cubic>]}
         BlueCurve {keys [<at 0 0 RCIM_Cubic>,<at 1 0 RCIM_Cubic>]}
         GreenCurve {keys [<at 0 0 RCIM_Cubic>,<at 1 0.0367 RCIM_Cubic>]}
         RedCurve {keys [<at 0 1 RCIM_Cubic>,<at 1 1 RCIM_Cubic>]}
        }})
      LerpFactor=$User.VariantIndex
     )
     Scale Factor=&MakeVector4FromVectorAndFloat(
      "Vector (RGB)"=<1 1 1>
      "Float (Alpha)"=&MaskFloatBySpawnGroup(
       Masked Float=&FloatFromCurve(FloatCurve=DataInterface() {propertyValues {Curve {keys [
            {interpMode=RCIM_Cubic, tangentMode=RCTM_User, time=0, value=1}
            {arriveTangent=-0.0487, interpMode=RCIM_Cubic, leaveTangent=-0.0487, tangentMode=RCTM_User, time=1, value=0}
           ]}}})
       Passthrough Float=&FloatFromCurve(FloatCurve=DataInterface() {propertyValues {Curve {keys [
            <at 0 0 RCIM_Cubic>
            {interpMode=RCIM_Cubic, tangentMode=RCTM_User, time=0.5, value=1}
            {arriveTangent=-0.0487, interpMode=RCIM_Cubic, leaveTangent=-0.0487, tangentMode=RCTM_User, time=1, value=0}
           ]}}})
      )
     )
    )
    Create Color Attribute=true
    Create Scale Attribute=true
    Scale=&Multiply_VectorByFloat(Vector=$Particles.Scale, Float=1)
    Write to Module Namespace=true
   ) {module=FN_Mesh_Attributes001}
   &FN_Light_Attributes(
    Color=<rgba 0.0563 0.2179 1 1>
    Create Color Attribute=true
    Create Exponent Attribute=true
    Create Radius Attribute=true
    Exponent=1.5
    Radius=&FloatFromCurve(
     FloatCurve=DataInterface() {propertyValues {Curve {keys [
         {interpMode=RCIM_Cubic, tangentMode=RCTM_User, time=0, value=1}
         {arriveTangent=-2.2724, interpMode=RCIM_Cubic, leaveTangent=-2.2724, tangentMode=RCTM_User, time=1, value=0}
        ]}}}
     Scale Curve=85
    )
    Scale Color=&FloatFromCurve(FloatCurve=DataInterface() {propertyValues {Curve {keys [<at 0.2 0>,<at 0.25 1>,<at 1 0>]}}}, Scale Curve=75)
   )
  ]
  rendererClasses [
   &NiagaraSpriteRendererProperties
   &NiagaraMeshRendererProperties
   &NiagaraLightRendererProperties
  ]
  renderers [
   {rendererClass=&NiagaraSpriteRendererProperties}
   {rendererClass=&NiagaraSpriteRendererProperties, rendererIndex=1}
   {rendererClass=&NiagaraSpriteRendererProperties, rendererIndex=2}
   {rendererClass=&NiagaraMeshRendererProperties, rendererIndex=3}
   {rendererClass=&NiagaraMeshRendererProperties, rendererIndex=4}
   {rendererClass=&NiagaraLightRendererProperties, rendererIndex=5}
  ]
 }
 NE_FlashBurst {
  emitterUpdateScript [
   &EmitterState()
   &SpawnPerFrame(
    Spawn:NiagaraBool=&Boolean_AndOperation(
     Bool1=&Boolean_NotOperation(Bool=$System.UpdateValueStateChange.Changed)
     Bool2=&Boolean_NotOperation(Bool=$Transient.FirstFrame)
    )
    Use Spawn Probability=false
   )
  ]
  particleSpawnScript [
   &InitializeParticle(
    Color Mode=::NewEnumerator0
    Lifetime=0.2
    Mesh Scale Mode=::NewEnumerator3
    Position Offset=$System.BaseOffset
    Sprite Rotation Mode=::NewEnumerator2
    Sprite Size Mode=::NewEnumerator3
    Uniform Sprite Size=50
    UsePositionOffset=true
    Write Material Random=true
   )
   /SetVariables(Particles.Age:NiagaraFloat=&Multiply_Float(A=$System.BuildUpTime, B=-0.9)) {isSetParameters=true, module=SetVariables_#6}
   &InitialMeshOrientation(Orientation Axis=<0 0 1>, Orientation Vector=<0 0 1>, Rotation=<0 0.25 0>, Use Orientation Vector=true, Use Rotation Vector=true)
   &InitialMeshOrientation(
    Orientation Axis=<0 0 0>
    Orientation Vector=<0 0 0>
    Rotation=&RandomRangeVector(Minimum=<0 0 0>, Maximum=<1 0 0>)
    Use Orientation Vector=false
    Use Rotation Vector=true
   ) {module=InitialMeshOrientation001}
   &CameraOffset(Camera Offset Amount=20)
  ]
  particleUpdateScript [
   &ParticleState()
   &ScaleSpriteSize(Uniform Curve Sprite Scale=DataInterface() {propertyValues {Curve {keys [
        {arriveTangent=3.2375, interpMode=RCIM_Cubic, leaveTangent=3.2375, tangentMode=RCTM_User, time=0, value=0}
        {arriveTangent=1, interpMode=RCIM_Cubic, leaveTangent=1, time=0.1, value=0.75}
        {interpMode=RCIM_Cubic, tangentMode=RCTM_User, time=1, value=1}
       ]}}})
   &ScaleColor(
    Scale Alpha=&MaskFloatBySpawnGroup(
     Masked Float=&FloatFromCurve(FloatCurve=DataInterface() {propertyValues {Curve {keys [
          {arriveTangent=-0.8571, interpMode=RCIM_Cubic, leaveTangent=-0.8571, tangentMode=RCTM_User, time=0, value=1}
          {arriveTangent=0.062, interpMode=RCIM_Cubic, leaveTangent=0.062, tangentMode=RCTM_User, time=1, value=0}
         ]}}})
     Passthrough Float=&FloatFromCurve(
      FloatCurve=DataInterface() {propertyValues {Curve {keys [
          {arriveTangent=0.9089, interpMode=RCIM_Cubic, leaveTangent=0.9089, tangentMode=RCTM_User, time=0, value=0}
          {arriveTangent=0.062, interpMode=RCIM_Cubic, leaveTangent=0.062, tangentMode=RCTM_User, time=0.5, value=1}
          <at 1 0 RCIM_Cubic>
         ]}}}
      Scale Curve=0.1
     )
    )
    Scale RGB=&MakeVectorFromLinearColorRGB(Color=&Multiply_LinearColorByFloat(
      Linear Color=&ColorFromCurve(ColorCurve=DataInterface() {propertyValues {
         AlphaCurve {keys [<at 0 1 RCIM_Cubic>]}
         BlueCurve {keys [<at 0 20 RCIM_Cubic>,<at 0.9851 10 RCIM_Cubic>]}
         GreenCurve {keys [<at 0 5.4413 RCIM_Cubic>,<at 0.9851 3.0139 RCIM_Cubic>]}
         RedCurve {keys [<at 0 0 RCIM_Cubic>,<at 0.9851 0.5952 RCIM_Cubic>]}
        }})
      Float=1
     ))
    ScaleA=true
    ScaleRGB=true
   )
   /SetVariables(
    Particles.GlowColor=&MultiplyLinearColor(
     Linear Color=$Particles.Color
     Scale Factor=&MakeVector4_FromLinearColor(LinearColor=&LerpLinearColors(StartColor=<rgba 0 0.014 0.5 1>, EndColor=<rgba 0.5 0.2545 0.2689 1>, LerpFactor=$User.VariantIndex))
    )
    Particles.GlowSize=&Multiply_Vector2D_ByFloat(Float=1.25, Vector2D=$Particles.SpriteSize)
   ) {isSetParameters=true, module=SetVariables_#7}
   /SetVariables(Particles.LongFlareSize:Vector2f=&Multiply_Vector2D(A=$Particles.SpriteSize, B=<2 2>)) {isSetParameters=true, module=SetVariables_#8}
   &DynamicMaterialParameters(Module.Param0WriteEnabled=false, Module.Param1WriteEnabled=false, Module.Param2WriteEnabled=false, Module.Param3WriteEnabled=false)
   &FN_Light_Attributes(
    Color=<rgba 0.0563 0.2179 1 1>
    Create Color Attribute=true
    Create Exponent Attribute=true
    Create Radius Attribute=true
    Exponent=1.5
    Radius=&FloatFromCurve(
     FloatCurve=DataInterface() {propertyValues {Curve {keys [
         {interpMode=RCIM_Cubic, tangentMode=RCTM_User, time=0, value=1}
         {arriveTangent=-2.2724, interpMode=RCIM_Cubic, leaveTangent=-2.2724, tangentMode=RCTM_User, time=1, value=0}
        ]}}}
     Scale Curve=85
    )
    Scale Color=&FloatFromCurve(FloatCurve=DataInterface() {propertyValues {Curve {keys [<at 0.2 0>,<at 0.25 1>,<at 1 0>]}}}, Scale Curve=75)
   )
  ]
  rendererClasses [&NiagaraSpriteRendererProperties, &NiagaraLightRendererProperties]
  renderers [
   {rendererClass=&NiagaraSpriteRendererProperties}
   {rendererClass=&NiagaraSpriteRendererProperties, rendererIndex=1}
   {rendererClass=&NiagaraLightRendererProperties, rendererIndex=2}
  ]
 }
 NE_SmokeDebris {
  emitterUpdateScript [
   &EmitterState()
   &SpawnPerFrame(
    Spawn:NiagaraBool=&Boolean_AndOperation(
     Bool1=&Boolean_NotOperation(Bool=$System.UpdateValueStateChange.Changed)
     Bool2=&Boolean_NotOperation(Bool=$Transient.FirstFrame)
    )
    Spawn Count=10
    Use Spawn Probability=false
   )
  ]
  particleSpawnScript [
   &InitializeParticle(
    Lifetime=&RandomRangeFloat(Minimum=0.8, Maximum=1.2)
    Sprite Rotation Mode=::NewEnumerator2
    Sprite Size Mode=::NewEnumerator3
    Uniform Sprite Size=&RandomRangeFloat(Minimum=30, Maximum=20)
    UsePositionOffset=false
    Write Material Random=true
   )
   /SetVariables(Particles.Age:NiagaraFloat=&Multiply_Float(A=$System.BuildUpTime, B=-1)) {isSetParameters=true}
   /SetVariables(Particles.VisibilityTag=&SpawnGroup()) {isSetParameters=true}
   &ShapeLocation(MaskSpawnGroup=false, Offset=$System.BaseOffset, Offset Mode=::NewEnumerator0, Override Seed=false, Sphere Radius=10, UseOwnerScale=true)
  ]
  particleUpdateScript [
   &ParticleState()
   &ScaleSpriteSize(Uniform Curve Sprite Scale=DataInterface() {propertyValues {Curve {keys [
        {arriveTangent=5.0791, interpMode=RCIM_Cubic, leaveTangent=5.0791, tangentMode=RCTM_User, time=0, value=0}
        {interpMode=RCIM_Cubic, tangentMode=RCTM_User, time=0.25, value=1}
       ]}}})
   /SetVariables(Particles.Position=&AddVectorToPosition(
     Position=$Particles.Initial.Position
     Vector=&Lerp_Vector(
      A=<0 0 0>
      B=&RandomRangeVector(Minimum=<0 0 -10>, Maximum=<1 1 -50>)
      Alpha=&FloatFromCurve(FloatCurve=DataInterface() {propertyValues {Curve {keys [
           {interpMode=RCIM_Cubic, tangentMode=RCTM_User, time=0, value=0}
           {arriveTangent=2.3149, interpMode=RCIM_Cubic, leaveTangent=2.3149, tangentMode=RCTM_User, time=1, value=1}
          ]}}})
     )
    )) {isSetParameters=true}
   &SolveForcesAndVelocity(Clamp Velocity=false, Limit Acceleration=false)
   &ScaleColor(
    Scale Alpha=&FloatFromCurve(FloatCurve=DataInterface() {propertyValues {Curve {keys [
         <at 0 0 RCIM_Cubic>
         {interpMode=RCIM_Cubic, tangentMode=RCTM_User, time=0.1, value=1}
         {arriveTangent=-1.0317, interpMode=RCIM_Cubic, leaveTangent=-1.0317, tangentMode=RCTM_User, time=1, value=0}
        ]}}})
    Scale RGB=&MakeVectorFromLinearColorRGB(Color=&ColorFromCurve(ColorCurve=DataInterface() {propertyValues {
        AlphaCurve {keys [<at 0 1 RCIM_Cubic>]}
        BlueCurve {keys [<at 0 3 RCIM_Cubic>,<at 0.2274 0.8 RCIM_Cubic>]}
        GreenCurve {keys [<at 0 2.2964 RCIM_Cubic>,<at 0.2274 0.2862 RCIM_Cubic>]}
        RedCurve {keys [<at 0 1.8178 RCIM_Cubic>,<at 0.2274 0.175 RCIM_Cubic>]}
       }}))
    ScaleA=true
    ScaleRGB=true
   )
   &SubUVAnimation(End Frame:NiagaraInt32=&Add_Integer(A=0, B=63))
   &FN_Sprite_Attributes(
    Color=<rgba 0.4186 0.6045 1 1>
    Create Color Attribute=true
    Create SpriteSize Attribute=true
    SpriteSize=&Vector2DFromFloat(Value=&Multiply_Float(
      A=&RandomRangeFloat(Minimum=4, Maximum=5)
      B=&FloatFromCurve(FloatCurve=DataInterface() {propertyValues {Curve {keys [
           {interpMode=RCIM_Cubic, tangentMode=RCTM_User, time=0, value=0}
           {interpMode=RCIM_Cubic, tangentMode=RCTM_User, time=0.1, value=1}
           {arriveTangent=-2.9382, interpMode=RCIM_Cubic, leaveTangent=-2.9382, tangentMode=RCTM_User, time=1, value=0}
          ]}}})
     ))
    Write to Module Namespace=true
   )
   /SetVariables(Particles.DebrisSubUV:NiagaraFloat=&RandomRangeFloat(Minimum=0, Maximum=3)) {isSetParameters=true}
  ]
  rendererClasses [&NiagaraSpriteRendererProperties]
  renderers [{rendererClass=&NiagaraSpriteRendererProperties}]
 }
}
summary {
 systemName=NS_MoleAcorn_BackPack_Idle
 emitters [
  {
   emitterName=NE_Flare
   rendererClasses [
    &NiagaraSpriteRendererProperties
    &NiagaraMeshRendererProperties
    &NiagaraLightRendererProperties
   ]
  }
  {emitterName=NE_ExplosionLightnings, rendererClasses [&NiagaraMeshRendererProperties]}
  {emitterName=NE_SmokeDebris, rendererClasses [&NiagaraSpriteRendererProperties]}
  {emitterName=NE_FlashBurst, rendererClasses [&NiagaraSpriteRendererProperties, &NiagaraLightRendererProperties]}
 ]
 userVariables [
  User.KillCount {type=NiagaraInt32}
  User.VariantIndex {type=NiagaraFloat, defaultValue=0, description=""}
 ]
}
