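# Material function for the MoleAcorn Pecan backpack's Booma effect, adding a glowing plasma look and a timed shrink reaction to incoming material attributes.
# Emissive fakes inner volume with layered panning noise at BumpOffset parallax depths, combined with Fresnel, flickering clouds and normal ripples.
# World position offset pulls vertices toward a fixed local point after the BP_GetTime timestamp, then scales them back up after a delay.
# Base color, roughness and specular are adjusted, optionally masked by a vertex color channel.
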MaterialFunction /BRCosmetics/Accessories/FORT_Backpacks/Backpack_M_MED_MoleAcornPecan/Materials/MF_MoleAcorn_Pecan_BoomaEffect
graph {nodes [
  Constant()
  n1 = NamedRerouteUsage()
  n2 = GetMaterialAttributes(None=NamedRerouteUsage()) {attributeGetTypes [#1, #2, #3]}
  n3 = LocalPosition() {localOrigin=InstancePreSkinning}
  n4 = Subtract(A=Constant3Vector() {constant=<rgba 0 -16 0 0>}, B=ShaderStageSwitch(PixelShader=VertexInterpolator(VS=@n3.XYZ).PS, VertexShader=@n3.XYZ))
  n5 = Subtract(
   A=Subtract(A=Time(), B=ScalarParameter() {defaultValue=9.999999843067494e+17, group=Booma, parameterName=BP_GetTime, sortPriority=10000})
   B=ScalarParameter() {defaultValue=1.1, group=Booma, parameterName=Reactivity_Delay}
  )
  n6 = Multiply(A=@n5) {constB=10}
  n7 = Subtract(A=@n5, B=ScalarParameter() {defaultValue=2.85, group=Booma, parameterName=ReactivityScaleUp_Delay}) {constB=5}
  n8 = OneMinus(None=Saturate(None=Multiply(A=@n7) {constB=1.75}))
  n9 = Power(Base=@n8) {constExponent=1.6}
  n10 = Saturate(None=Subtract(A=Saturate(None=@n6), B=OneMinus(None=@n9)))
  n11 = SetMaterialAttributes!(
   MaterialAttributes=@n2.MaterialAttributes
   EmissiveColor=NamedRerouteUsage()
   BaseColor=NamedRerouteUsage()
   WorldPositionOffset=Transform(None=LinearInterpolate(B=@n4, Alpha=@n10) {constB=-1}) {transformSourceType=TRANSFORMSOURCE_Local}
   Roughness=NamedRerouteUsage()
   Specular=NamedRerouteUsage()
  ) {attributeSetTypes [
    #4
    #1
    #5
    #2
    #3
   ]}
  n12 = ChannelMaskParameter(None=VertexColor()) {group=Booma, parameterName=MoleAcorn_VertexColorChannel, sortPriority=60}
  FunctionOutput(None=StaticSwitchParameter(True=BlendMaterialAttributes(A=@n1, B=@n11, Alpha=@n12.R), False=@n11) {group=Booma, parameterName=MoleAcorn_UseVertexColorMask, sortPriority=60})
  MaterialFunctionCall() {materialFunction=&WorldAlignedNormal}
  n13 = PreSkinnedNormal()
  n14 = ShaderStageSwitch(PixelShader=VertexInterpolator(VS=@n13).PS, VertexShader=@n13)
  n15 = Saturate(None=Subtract(A=Abs(None=@n14)) {constB=0.8})
  n16 = ComponentMask(None=Saturate(None=Multiply(A=@n15) {constB=5.5})) {b=true}
  Multiply(A=Multiply(A=@n16, B=VectorParameter() {group=Booma, parameterName=Booma_Color, defaultValue=<rgba 1 0.1174 0.1728 1>}.RGB)) {constB=0.4}
  Multiply(A=@n8, B=@n8)
  n17 = ComponentMask(None=NamedRerouteUsage()) {g=true, r=true}
  n18 = TextureCoordinate()
  n19 = BumpOffset(
   Coordinate=Multiply(A=Add(A=@n18) {constB=0.22}, B=ScalarParameter() {defaultValue=10, group=Booma, parameterName=TextureScale_04, sortPriority=60}) {constB=10}
   Height=ScalarParameter() {defaultValue=-30, group=Booma, parameterName=BumpOffset_04, sortPriority=60}
  )
  n20 = VectorParameter() {group=Booma, parameterName=BoomaTextures_PannerSpeed, sortPriority=60, defaultValue=<rgba 0 -0.075 0 1>}
  n21 = TextureSample(UVs=Add(A=@n17, B=Panner(Coordinate=@n19, Speed=@n20.RGB) {speedX=0.1})) {
   samplerSource=SSM_Wrap_WorldGroupSettings
   texture=/BRCosmetics/Accessories/FORT_Backpacks/Backpack_M_MED_MoleAcornPecan/Textures/T_MoleAcorn_Noise
  }
  n22 = Multiply(
   A=Add(A=TextureCoordinate()) {constB=0.22}
   B=ScalarParameter() {defaultValue=6, group=Booma, parameterName=TextureScale_05, sortPriority=60}
  ) {constB=10}
  n23 = ScalarParameter() {defaultValue=-10, group=Booma, parameterName=BumpOffset_05, sortPriority=60}
  n24 = TextureSample(UVs=Panner(Coordinate=BumpOffset(Coordinate=@n22, Height=@n23), Speed=@n20.RGB) {speedX=0.1}) {
   samplerSource=SSM_Wrap_WorldGroupSettings
   texture=&T_MoleAcorn_Noise
  }
  n25 = Multiply(A=Add(A=@n21.R, B=@n24.R), B=VectorParameter() {group=Booma, parameterName=Booma_Color_03, defaultValue=<rgba 1 0.5815 0.9168 1>}.RGB)
  n26 = Constant() {r=-2.7}
  n27 = ComponentMask(None=NamedRerouteUsage()) {g=true, r=true}
  n28 = TextureSample(UVs=Add(A=BumpOffset(Height=@n26), B=Multiply(A=@n27) {constB=0.5})) {
   samplerSource=SSM_Wrap_WorldGroupSettings
   texture=&T_MoleAcorn_Noise
  }
  n29 = Saturate(None=Multiply(A=@n28.B) {constB=10})
  n30 = Add(A=@n28.B, B=Multiply(A=Time()) {constB=0.5})
  n31 = Add(A=Multiply(A=Sine(None=@n30)) {constB=0.5}) {constB=0.5}
  n32 = Multiply(
   A=VectorParameter() {group=Booma, parameterName=Booma_Color_04, defaultValue=<rgba 1 0.3059 0.2742 1>}.RGB
   B=Multiply(A=Multiply(A=@n29, B=@n31)) {constB=3}
  )
  n33 = VectorParameter() {group=Booma, parameterName=Booma_Color_01, defaultValue=<rgba 1 0.1174 0.1728 1>}
  n34 = BumpOffset(
   Coordinate=Multiply(A=TextureCoordinate(), B=ScalarParameter() {defaultValue=0.27, group=Booma, parameterName=TextureScale_01, sortPriority=60}) {constB=0.27}
   Height=ScalarParameter() {defaultValue=-3, group=Booma, parameterName=BumpOffset_01, sortPriority=60}
  )
  n35 = TextureSample(UVs=Add(A=Panner(Coordinate=@n34, Speed=@n20.RGB) {speedX=0.1}, B=@n17)) {
   samplerSource=SSM_Wrap_WorldGroupSettings
   texture=&T_MoleAcorn_Noise
  }
  n36 = TextureCoordinate()
  n37 = BumpOffset(
   Coordinate=Multiply(A=Add(A=@n36) {constB=0.2}, B=ScalarParameter() {defaultValue=0.27, group=Booma, parameterName=TextureScale_02, sortPriority=60}) {constB=0.5}
   Height=ScalarParameter() {defaultValue=-10, group=Booma, parameterName=BumpOffset_02, sortPriority=60}
  )
  n38 = TextureSample(UVs=Add(A=@n17, B=Panner(Coordinate=@n37, Speed=@n20.RGB) {speedX=0.1})) {
   samplerSource=SSM_Wrap_WorldGroupSettings
   texture=&T_MoleAcorn_Noise
  }
  n39 = Add(A=@n35.R, B=@n38.R)
  n40 = TextureCoordinate()
  n41 = BumpOffset(
   Coordinate=Multiply(A=Add(A=@n40) {constB=0.2}, B=ScalarParameter() {defaultValue=2, group=Booma, parameterName=TextureScale_03, sortPriority=60}) {constB=2}
   Height=ScalarParameter() {defaultValue=-20, group=Booma, parameterName=BumpOffset_03, sortPriority=60}
  )
  n42 = TextureSample(UVs=Add(A=@n17, B=Panner(Coordinate=@n41, Speed=@n20.RGB) {speedX=0.1})) {
   samplerSource=SSM_Wrap_WorldGroupSettings
   texture=&T_MoleAcorn_Noise
  }
  n43 = VectorParameter() {group=Booma, parameterName=Booma_Color_02, defaultValue=<rgba 0.403 0.2627 1 1>}
  n44 = Add(A=Add(A=Add(A=@n25, B=@n32), B=NamedRerouteUsage()), B=Add(A=Multiply(A=@n33.RGB, B=@n39), B=Multiply(A=@n42.R, B=@n43.RGB)))
  Multiply(A=@n44, B=@n33.RGB)
  n45 = ScalarParameter() {defaultValue=0.5, group=Booma, parameterName=NormalNoise_Intensity, sortPriority=60}
  n46 = TextureCoordinate()
  n47 = ScalarParameter() {defaultValue=1, group=Booma, parameterName=NormalNoise_Scale, sortPriority=60}
  n48 = TextureSample(UVs=Panner() {speedX=0.02, speedY=0.05}) {samplerSource=SSM_Wrap_WorldGroupSettings, samplerType=SAMPLERTYPE_Normal, texture=/Game/Global/Textures/Default/Noise/T_Normal_Noise01}
  n49 = TextureSample(UVs=Panner() {speedX=-0.01, speedY=-0.025}) {samplerSource=SSM_Wrap_WorldGroupSettings, samplerType=SAMPLERTYPE_Normal, texture=&T_Normal_Noise01}
  n50 = Add(A=ComponentMask(None=@n48.RGB) {g=true, r=true}, B=ComponentMask(None=@n49.RGB) {g=true, r=true})
  n51 = TextureSample(UVs=Add(A=Multiply(A=@n46, B=@n47), B=Multiply(A=@n50) {constB=0.05})) {
   samplerSource=SSM_Wrap_WorldGroupSettings
   samplerType=SAMPLERTYPE_Normal
   texture=/BRCosmetics/Accessories/FORT_Backpacks/Backpack_M_MED_MoleAcornPecan/Textures/T_MoleAcorn_VolumeNormal
  }
  NamedRerouteDeclaration(None=Multiply(A=Divide(A=@n45) {constB=10}, B=@n51.RGB) {constA=0.05, constB=0.1}) {name=NormalWripples, nodeColor=<rgba 0 0.7188 1 1>}
  n52 = NamedRerouteUsage()
  n53 = Constant3Vector() {constant=<rgba 0 0.5 0.1 0>}
  n54 = FunctionInput(Preview=MakeMaterialAttributes!(
    BaseColor=Constant3Vector() {constant=<rgba 0.0685 0.1529 0.3968 1>}
    Metallic=@n53.R
    Specular=@n53.R
    Roughness=@n53.R
    Anisotropy
    EmissiveColor
    Opacity
    OpacityMask
    Normal
    Tangent
    WorldPositionOffset
    SubsurfaceColor
    ClearCoat
    ClearCoatRoughness
    AmbientOcclusion
    Refraction
    CustomizedUVs_0
    CustomizedUVs_1
    CustomizedUVs_2
    CustomizedUVs_3
    CustomizedUVs_4
    CustomizedUVs_5
    CustomizedUVs_6
    CustomizedUVs_7
    PixelDepthOffset
    ShadingModel
    Displacement
   )) {inputName=MA, inputType=FunctionInput_MaterialAttributes}
  n55 = VectorParameter() {group=PBR, parameterName=Preview_Diffuse, sortPriority=0, defaultValue=<rgba 0.0685 0.1529 0.3968 1>}
  n56 = Constant() {r=1}
  n57 = Constant() {r=0.7}
  n58 = VectorParameter() {group=PBR, parameterName=Preview_SRM, defaultValue=<rgba 0.3 0.55 0 1>}
  n59 = SetMaterialAttributes!(
   MaterialAttributes=NamedRerouteDeclaration(None=@n54) {name=OriginalMaterialAttributes, nodeColor=<rgba 0 1 0.0859 1>}
   BaseColor=Multiply(
    A=@n55.RGB
    B=MaterialFunctionCall(Value_Morning=@n56, Value_Day=@n56, Value_Evening=@n56, Value_Night=@n57) {materialFunction=&Scalar_Time_of_Day}.Result
   )
   Metallic=@n58.B
   Specular=@n58.B
   Roughness=@n58.B
  ) {attributeSetTypes [
    #1
    #6
    #3
    #2
   ]}
  n60 = ChannelMaskParameter(None=VertexColor()) {group=Booma, parameterName=MoleAcorn_VertexColorChannel, sortPriority=60}
  NamedRerouteDeclaration(None=StaticSwitchParameter(True=BlendMaterialAttributes(A=@n52, B=@n59, Alpha=@n60.R), False=@n59) {group=Booma, parameterName=MoleAcorn_UseVertexColorMask, sortPriority=60}) {name=MaterialAttributes_IN, nodeColor=<rgba 0.6406 0 1 1>}
  NamedRerouteDeclaration(None=Add(A=@n2.Specular, B=ScalarParameter() {defaultValue=0.1, group=Booma, parameterName=MoleAcpm_Spec})) {name=Spec_outout, nodeColor=<rgba 0 1 0.1562 1>}
  n61 = Multiply(
   A=Multiply(
    A=NamedRerouteUsage()
    B=VectorParameter() {group=Booma, parameterName=Booma_Color_AdditionalGlow, defaultValue=<rgba 1 0.1174 0.1728 1>}.RGB
   )
   B=ScalarParameter() {defaultValue=1, group=Booma, parameterName=AdditionalGlow_Amount, sortPriority=60}
  ) {constB=1.5}
  n62 = NamedRerouteUsage()
  n63 = Multiply(
   A=Multiply(A=Multiply(A=@n62, B=@n44), B=ScalarParameter() {defaultValue=5, group=Booma, parameterName=BoomaGlow_Intensity}) {constB=0.5}
   B=NamedRerouteUsage()
  )
  n64 = Add(B=OneMinus(None=@n16)) {constA=0.01}
  NamedRerouteDeclaration(None=LinearInterpolate(A=Multiply(A=@n33.RGB) {constB=0.01}, B=Add(A=@n61, B=@n63), Alpha=Saturate(None=@n64))) {name=Emissive_OUT, nodeColor=<rgba 0 1 0.7188 1>}
  n65 = Saturate(None=Fresnel() {exponent=2.2})
  n66 = ScalarParameter() {defaultValue=-0.3, group=PBR, parameterName=BaseColorFresne_Thickness, sliderMax=1, sortPriority=50}
  n67 = Multiply(
   A=Saturate(None=Add(A=@n65, B=@n66) {constB=0.5})
   B=ScalarParameter() {defaultValue=1, group=PBR, parameterName=BaseColorFresne_Intensity, sliderMax=1, sortPriority=50}
  )
  n68 = Add(A=ComponentMask(None=VertexNormalWS()) {b=true}) {constB=0.3}
  n69 = Saturate(None=Add(A=Saturate(None=@n68), B=ScalarParameter() {defaultValue=0.1, group=PBR, parameterName=Booma_ZNormal_Darkening_Mult}))
  NamedRerouteDeclaration(None=Add(A=Multiply(A=@n2.BaseColor, B=@n67), B=Multiply(A=@n2.BaseColor, B=@n69))) {name=BaseColor_OUT, nodeColor=<rgba 1 0.3984 0 1>}
  n70 = ScalarParameter() {defaultValue=0.85, group=Booma, parameterName=Booma_FresnelExponent}
  n71 = Multiply(A=Fresnel(ExponentIn=@n70) {exponent=0.6}) {constB=1.5}
  NamedRerouteDeclaration(None=Power(Base=OneMinus(None=@n71), Exp=ScalarParameter() {defaultValue=6, group=Booma, parameterName=AdditionalGlow_Power, sortPriority=60}) {constExponent=0.07}) {name=Fresnel_AdditionalGlow, nodeColor=<rgba 0.7813 0 1 1>}
  n72 = MaterialFunctionCall() {materialFunction=/BRCosmetics/Accessories/FORT_Backpacks/Backpack_F_MED_GroovyReader_Play/Materials/MaterialFunctions/MF__FloatFlickerRandom}
  n73 = Saturate(None=Add(A=@n72.RandomFlickerFloat) {constB=0.8})
  n74 = TextureCoordinate()
  n75 = ComponentMask(None=Multiply(A=NamedRerouteUsage()) {constB=0.3}) {g=true, r=true}
  n76 = TextureSample(UVs=BumpOffset(Coordinate=Add(A=@n74, B=@n75), Height=Constant() {r=-7})) {
   samplerSource=SSM_Wrap_WorldGroupSettings
   texture=/BRCosmetics/Accessories/FORT_Backpacks/Backpack_M_MED_MoleAcornPecan/Textures/T_MoleAcorn_Noise_02
  }
  n77 = Add(A=@n76.G, B=Multiply(A=Time()) {constB=0.31})
  n78 = Add(A=Multiply(A=Sine(None=@n77)) {constB=0.5}) {constB=0.5}
  n79 = Multiply(A=Power(Base=Multiply(A=@n76.R, B=@n78)), B=ScalarParameter() {defaultValue=10, group=Booma, parameterName=Clouds_Brightness}) {constB=10}
  NamedRerouteDeclaration(None=Multiply(A=VectorParameter() {group=Booma, parameterName=Booma_Color_05, defaultValue=<rgba 1 0.1174 0.1728 1>}.RGB, B=Multiply(A=@n73, B=@n79))) {name=BrightClouds, nodeColor=<rgba 0.125 1 0 1>}
  n80 = TextureSample(UVs=TextureCoordinate() {uTiling=2, vTiling=2}) {
   samplerSource=SSM_Wrap_WorldGroupSettings
   samplerType=SAMPLERTYPE_Normal
   texture=&T_MoleAcorn_VolumeNormal
  }
  n81 = Transform(None=MaterialFunctionCall(Normal=@n80.RGB, Flatness=Constant() {r=1}) {materialFunction=&FlattenNormal}.Result)
  n82 = Saturate(None=Multiply(
    A=Fresnel(ExponentIn=@n70, Normal=@n81) {exponent=0.6}
    B=ScalarParameter() {defaultValue=1, group=Booma, parameterName=Booma_FresnelShrinking}
   ) {constB=2})
  NamedRerouteDeclaration(None=Power(Base=OneMinus(None=@n82), Exp=ScalarParameter() {defaultValue=7, group=Booma, parameterName=Booma_FresnelSoftness}) {constExponent=7}) {name=Fresnel_InnerPlasma, nodeColor=<rgba 0 1 0.4609 1>}
  n83 = Multiply(A=TextureCoordinate(), B=ScalarParameter() {defaultValue=2, group=Booma, parameterName=NoiseCloud_Scale})
  n84 = Multiply(A=ComponentMask(None=NamedRerouteUsage()) {g=true, r=true}) {constB=0.3}
  n85 = TextureSample(UVs=BumpOffset(Coordinate=Add(A=@n83, B=@n84), Height=Constant() {r=-2.5})) {
   samplerSource=SSM_Wrap_WorldGroupSettings
   texture=&T_MoleAcorn_Noise
  }
  n86 = Saturate(None=Multiply(A=Power(Base=@n85.G) {constExponent=12}) {constB=2})
  n87 = OneMinus(None=ScalarParameter() {defaultValue=0.98, group=Booma, parameterName=NoiseCloud_Intensity})
  NamedRerouteDeclaration(None=Saturate(None=Add(A=@n86, B=@n87) {constB=0.01})) {name=NoiseCloud, nodeColor=<rgba 0.2188 0 1 1>}
  NamedRerouteDeclaration(None=Add(A=@n2.Roughness, B=ScalarParameter() {defaultValue=-0.18, group=Booma, parameterName=MoleAcpm_roughness})) {name=Roughness_outout, nodeColor=<rgba 0 1 0.1562 1>}
  Reroute(None=@n54)
  SmoothStep() {constMin=0.2}
 ]}
